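# Conversation database for the BountyPond NPC, a KPop Demon Hunters-themed character, defining its greeting and top-level conversation branches.
# A single entry point feeds a Speech node that has a default line plus contextual greetings gated by first-ever meeting, match conversation count, or the player's equipped outfit, with curve-based priorities for first-meeting and outfit reactions.
# The speech then links out to remote Quests, BasicServices and BountyPond Dialogue entry points.
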ConversationDatabase /NPCLibrary/NPCs/BountyPond/NPCLibrary_Conversation_BountyPond
entry_points [{tag=Conversation.Entry.BountyPond, destinations [#1]}]
nodes [
 ConversationEntryPointNode_7 {
  id=#1
  class=ConversationEntryPointNode
  outputs [#2]
  properties {entryTag {tagName=Conversation.Entry.BountyPond}}
 }
 Speech_C_4 {
  id=#2
  class=Speech_C
  outputs [#3, #4, #5]
  properties {
   nodeName=""
   generalConfig {
    defaultMessage=I might just be looking for a little soda pop... pop.
    contextualMessages [
     {
      message=Heh. Looking for an idol?
      contextRequirements [{requirement {class=FortControllerRequirement_IsFirstEverConversationWithNPC}}]
      priority {curve {rowName=Tandem.Speech.Priority.FirstEverInteraction}}
     }
     {
      message=The Saja Boys will always end up on top.
      contextRequirements [{requirement {class=FortControllerRequirement_CheckNPCMatchConversationCount, properties {nPCUniqueIdTag {tagName=AISpawnerData.Type.Tandem.BountyPond}}}}]
      priority {curve {rowName=None}}
     }
     {
      message=Watch yourself.
      contextRequirements [{requirement {class=FortControllerRequirement_CheckNPCMatchConversationCount, properties {nPCUniqueIdTag {tagName=AISpawnerData.Type.Tandem.BountyPond}}}}]
      priority {curve {rowName=None}}
     }
     {
      message=I might just be looking for a little soda pop... pop.
      contextRequirements [{requirement {class=FortControllerRequirement_CheckNPCMatchConversationCount, properties {nPCUniqueIdTag {tagName=AISpawnerData.Type.Tandem.BountyPond}}}}]
      priority {curve {rowName=None}}
     }
     {
      message=You lost one family. Is HUNTR/X gonna abandon you too?
      contextRequirements [{requirement {
         class=FortControllerRequirement_HasCID
         properties {allowedCIds [/BRCosmetics/Athena/Items/Cosmetics/Characters/Character_QuailWink, /BRCosmetics/Athena/Items/Cosmetics/Characters/Character_LowerDrop]}
        }}]
      priority {curve {rowName=Tandem.Speech.Priority.OutfitReaction}}
     }
     {
      message="Have you figured out who you are yet, or are you hiding behind HUNTR/X?"
      contextRequirements [{requirement {
         class=FortControllerRequirement_HasCID
         properties {allowedCIds [/BRCosmetics/Athena/Items/Cosmetics/Characters/Character_EvokeFind, /BRCosmetics/Athena/Items/Cosmetics/Characters/Character_HybridBoot]}
        }}]
      priority {curve {rowName=Tandem.Speech.Priority.OutfitReaction}}
     }
     {
      message="You'll always be a Hunter who's part demon, Rumi."
      contextRequirements [{requirement {
         class=FortControllerRequirement_HasCID
         properties {allowedCIds [/BRCosmetics/Athena/Items/Cosmetics/Characters/Character_LoanFloat, /BRCosmetics/Athena/Items/Cosmetics/Characters/Character_TurnVortex]}
        }}]
      priority {curve {rowName=Tandem.Speech.Priority.OutfitReaction}}
     }
     {
      message=HUNTR/X will always wipe the floor with the Saja Boys.
      contextRequirements [{requirement {class=FortControllerRequirement_HasCID, properties {allowedCIds [/BRCosmetics/Athena/Items/Cosmetics/Characters/Character_BountyPond]}}}]
      priority {curve {rowName=Tandem.Speech.Priority.OutfitReaction}}
     }
    ]
   }
  }
 }
 ConversationLinkNode_2 {
  id=#5
  class=ConversationLinkNode
  properties {remoteEntryTag {tagName=Conversation.Entry.BountyPond.Dialogue}}
 }
 ConversationLinkNode_0 {id=#4, class=ConversationLinkNode, properties {remoteEntryTag {tagName=Conversation.Entry.BasicServices}}}
 ConversationLinkNode_1 {id=#3, class=ConversationLinkNode, properties {remoteEntryTag {tagName=Conversation.Entry.NPCLibrary.Quests}}}
]
