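# Conversation database for the BountyPond NPC's "Ask about..." dialogue, delivering character lore on the Saja Boys, HUNTR/X and his past.
# The entry tag Conversation.Entry.BountyPond.Dialogue leads to an opening speech that branches into those three topic choices plus a Back node.
# Each topic plays one or more chained speech lines, and all three converge on a shared ReturnToStart node shown as BACK.
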ConversationDatabase /NPCLibrary/NPCs/BountyPond/NPCLibrary_Conversation_BountyPond_Dialogue
entry_points [{tag=Conversation.Entry.BountyPond.Dialogue, destinations [#1]}]
nodes [
 ConversationEntryPointNode_0 {
  id=#1
  class=ConversationEntryPointNode
  outputs [#2]
  properties {entryTag {tagName=Conversation.Entry.BountyPond.Dialogue}}
 }
 Speech_C_0 {
  id=#2
  class=Speech_C
  outputs [#3, #4, #5, #6]
  properties {
   nodeName=""
   softTaskIcon=&T_UI_NPC_question
   generalConfig {defaultMessage="Alright, but every answer has a price..."}
  }
  sub_nodes [ConversationChoiceNode_5 {class=ConversationChoiceNode, properties {defaultChoiceDisplayText=Ask about...}}]
 }
 Speech_C_3 {
  id=#6
  class=Speech_C
  outputs [#7]
  properties {nodeName="", softTaskIcon=&T_UI_NPC_question, generalConfig {defaultMessage=I'm not gonna talk about that.}}
  sub_nodes [ConversationChoiceNode_1 {class=ConversationChoiceNode, properties {defaultChoiceDisplayText=His Past}}]
 }
 ReturnToStart_C_23 {
  id=#7
  class=ReturnToStart_C
  properties {softTaskIcon=&T_UI_NPC_Back}
  sub_nodes [ConversationChoiceNode_32 {class=ConversationChoiceNode, properties {defaultChoiceDisplayText=BACK}}]
 }
 Speech_C_2 {
  id=#5
  class=Speech_C
  outputs [#8]
  properties {nodeName="", softTaskIcon=&T_UI_NPC_question, generalConfig {defaultMessage="They're pretty good, I guess."}}
  sub_nodes [ConversationChoiceNode_2 {class=ConversationChoiceNode, properties {defaultChoiceDisplayText=HUNTR/X}}]
 }
 Speech_C_4 {
  id=#8
  class=Speech_C
  outputs [#7]
  properties {nodeName="", softTaskIcon=&T_UI_NPC_question, generalConfig {defaultMessage="We're better, though."}}
 }
 Back_C_0 {
  id=#4
  class=Back_C
  sub_nodes [ConversationChoiceNode_7 {class=ConversationChoiceNode, properties {defaultChoiceDisplayText=BACK}}]
 }
 Speech_C_7 {
  id=#3
  class=Speech_C
  outputs [#9]
  properties {
   nodeName=""
   softTaskIcon=&T_UI_NPC_question
   generalConfig {defaultMessage=My most brilliant plan: go after the one thing HUNTR/X wouldn't expect.}
  }
  sub_nodes [ConversationChoiceNode_0 {class=ConversationChoiceNode, properties {defaultChoiceDisplayText=The Saja Boys}}]
 }
 Speech_C_15 {
  id=#9
  class=Speech_C
  outputs [#10]
  properties {
   nodeName=""
   softTaskIcon=&T_UI_NPC_question
   generalConfig {defaultMessage=Their fans. The very thing that powers the Honmoon.}
  }
 }
 Speech_C_14 {
  id=#10
  class=Speech_C
  outputs [#11]
  properties {nodeName="", softTaskIcon=&T_UI_NPC_question, generalConfig {defaultMessage=I need to beat them. I need to win.}}
 }
 Speech_C_1 {
  id=#11
  class=Speech_C
  outputs [#7]
  properties {
   nodeName=""
   softTaskIcon=&T_UI_NPC_question
   generalConfig {defaultMessage=I need to forget the things I've done.}
  }
 }
]
